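# Opaque Default Lit material for the Demeter landscape's stone path spline props, enabled for spline meshes and Nanite.
# A shared rock diffuse texture is multiplied by a warm, muted constant color to tint it.
# Two scalar constants sit alongside the tint, and no parameters or other textures are used.
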Material /Game/Environments/Landscape/Demeter/Splines/Props/M_Demeter_SplineStonePath
graph {
 properties {bUsedWithNanite=true, bUsedWithSplineMeshes=true, pixelDepthOffsetMode=PDOM_Legacy, usedShadingModels=Default Lit}
 nodes [
  Constant() {r=0.4}
  Constant() {r=0.85}
  Multiply(
   A=TextureSample() {texture=/Game/Environments/Props/Rocks/Textures/Rocks_D}.RGB
   B=Constant3Vector() {constant=<rgba 0.6198 0.5331 0.4891 0>}
  )
 ]
}
